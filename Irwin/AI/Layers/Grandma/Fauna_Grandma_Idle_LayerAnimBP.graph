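# Animation layer Blueprint for the Grandma fauna predator that overrides its idle pose in the FaunaIdleOverride layer group.
# Fauna Idle outputs the Grandma_Idle_Break sequence directly, ignoring its input pose, while Fauna IdleTurn passes its input through.
# An inertialized blend between idle and a random break player, driven by bHasValidAimTarget, sits unconnected to the output.
# Initialization casts and caches the main Grandma_AnimBP, which that blend reads.
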AnimBlueprint /Irwin/AI/Layers/Grandma/Fauna_Grandma_Idle_LayerAnimBP
target_skeleton=/Irwin/AI/Base/Grandma_Mammal_Invis_Skeleton
graphs [
 AnimGraph {nodes [Output Pose() {layerGroup=None}]}
 EventGraph {nodes [
   Event BlueprintInitializeAnimation()
   > n1 = Cast To Grandma_AnimBP!(execute, Object=Get MainAnimBP() -> (MainAnimBP:Anim Instance Object Reference)) -> (then, CastFailed, AsGrandma Anim BP:Grandma Anim BP Object Reference) {is=title K2Node_DynamicCast}
   > Set GrandmaAnimBP(GrandmaAnimBP:Grandma Anim BP Object Reference=@n1) -> (Output_Get:Grandma Anim BP Object Reference)
  ]}
 Fauna Idle {nodes [
   n1 = Random Sequence Player!() -> (Pose:Pose Link Structure) {
    is=title AnimGraphNode_RandomPlayer
    entries [
     {
      chanceToPlay=1
      maxLoopCount=0
      maxPlayRate=1
      minLoopCount=0
      minPlayRate=1
      sequence=/Irwin/AI/Predators/Grandma/Animations/Grandma_Idle_Break
      blendIn {blendOption=HermiteCubic, blendTime=0.6, customCurve=null}
     }
     {
      chanceToPlay=1
      maxLoopCount=0
      maxPlayRate=1
      minLoopCount=0
      minPlayRate=1
      sequence=&Grandma_Idle_Break
      blendIn {blendOption=HermiteCubic, blendTime=0.6, customCurve=null}
     }
    ]
   }
   Blend Poses by bool(
    BlendPose_0=Sequence Player 'Grandma_Idle'() {sequence=/Irwin/AI/Predators/Grandma/Animations/Grandma_Idle}
    BlendPose_1=@n1
    BlendTime_0=0.4
    BlendTime_1=0.4
    bActiveValue=Get bHasValidAimTarget!(self:Fort AIFauna Anim Instance Object Reference=Get GrandmaAnimBP() -> (GrandmaAnimBP:Grandma Anim BP Object Reference)) -> (bHasValidAimTarget:bool)
   ) {blendType=Linear, transitionType=Inertialization}
   Input Pose() {name=InPose Idle}
   Output Pose(Result=Sequence Player 'Grandma_Idle_Break'() {sequence=&Grandma_Idle_Break}) {layerGroup=FaunaIdleOverride}
  ]}
 Fauna IdleTurn {nodes [Output Pose(Result=Input Pose() {name=InPose IdleTurn}) {layerGroup=FaunaIdleOverride}]}
]
variables [GrandmaAnimBP {type=Grandma_AnimBP_C, default=null}]
